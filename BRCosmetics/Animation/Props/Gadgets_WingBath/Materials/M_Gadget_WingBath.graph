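# Master material for the WingBath gadget emote prop, adding controllable heat glow to its jet, tongs and bar.
# An RGB part mask combines with pre-skinned-position gradients, each scaled by its own transition parameter, with noise breakup on the bar.
# The summed heat samples a gradient texture for emissive color and drives darkened base color and adjusted specular and roughness.
# These overrides sit between the 2023 CharacterParent PreFX and PostFX functions.
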Material /BRCosmetics/Animation/Props/Gadgets_WingBath/Materials/M_Gadget_WingBath
graph {
 properties {
  bUseMaterialAttributes=true
  bUsedWithMeshDeformer=true
  bUsedWithMorphTargets=true
  bUsedWithSkeletalMesh=true
  pixelDepthOffsetMode=PDOM_Legacy
  previewMesh=/BRCosmetics/Animation/Props/Gadgets_WingBath/Meshes/WingBath_Gadget
  shadingModel=MSM_FromMaterialExpression
  usedShadingModels=Default Lit | Subsurface | Clear Coat | Subsurface Profile | Cloth
 }
 nodes [
  n1 = MaterialFunctionCall() {materialFunction=&MF_CharacterParent_2023_PreFX}
  n2 = GetMaterialAttributes(None=@n1.Result) {attributeGetTypes [
    #1
    #2
    #3
    #4
   ]}
  n3 = NamedRerouteUsage()
  n4 = NamedRerouteUsage()
  n5 = SetMaterialAttributes!(
   MaterialAttributes=@n2.MaterialAttributes
   Base Color=LinearInterpolate(A=@n2.Base Color, Alpha=@n3) {constB=0}
   Specular=LinearInterpolate(A=@n2.Specular, Alpha=@n3)
   Roughness=LinearInterpolate(A=@n2.Roughness, Alpha=@n3)
   Emissive Color=LinearInterpolate(A=@n2.Emissive Color, B=Add(A=@n2.Emissive Color, B=@n4), Alpha=@n3)
  ) {attributeSetTypes [
    #1
    #2
    #3
    #4
   ]}
  MaterialFunctionCall(Material=@n5, M=@n1.M, Pre Skinned Local Position=@n1.M, Distance Blend=@n1.M, Fresnel=@n1.M) {materialFunction=&MF_CharacterParent_2023_PostFX}
  NamedRerouteDeclaration(None=VertexInterpolator(VS=LocalPosition() {localOrigin=InstancePreSkinning}.XYZ).PS) {name=PreSkinnedLoc, nodeColor=<rgba 1 0.9375 0 1>}
  n6 = TextureSampleParameter2D() {
   group=WingBath
   parameterName=PropMsk
   sortPriority=0
   texture=/BRCosmetics/Animation/Props/Gadgets_WingBath/Textures/T_WingBath_Emote_Prop_Tech_Msk
  }
  NamedRerouteDeclaration(None=@n6.R) {name=Jet_Msk, nodeColor=<rgba 0 1 0.2969 1>}
  NamedRerouteDeclaration(None=@n6.B) {name=Tongs_Msk, nodeColor=<rgba 1 0 0.0234 1>}
  NamedRerouteDeclaration(None=@n6.G) {name=Bar_Msk, nodeColor=<rgba 1 0 0.0234 1>}
  n7 = NamedRerouteUsage()
  n8 = Multiply(A=SmoothStep(Value=ComponentMask(None=@n7) {b=true}) {constMax=56.2, constMin=53.8}, B=NamedRerouteUsage())
  n9 = ScalarParameter() {defaultValue=1, group=WingBath, parameterName=HeatTransition_Jet}
  n10 = ComponentMask(None=NamedRerouteUsage()) {g=true}
  n11 = Multiply(A=SmoothStep(Value=Add(A=@n10) {constB=-30}) {constMax=25}, B=NamedRerouteUsage())
  n12 = ScalarParameter() {defaultValue=1, group=WingBath, parameterName=HeatTransition_Tongs}
  n13 = NamedRerouteUsage()
  n14 = SmoothStep(Value=ComponentMask(None=@n13) {g=true}) {constMax=10, constMin=-15}
  n15 = SmoothStep(Value=ComponentMask(None=@n13) {b=true}) {constMax=10, constMin=-10}
  n16 = Multiply(A=TextureCoordinate()) {constB=2}
  n17 = Subtract(
   A=Add(A=OneMinus(None=@n14), B=OneMinus(None=@n15))
   B=Multiply(
    A=TextureSample(UVs=@n16) {samplerType=SAMPLERTYPE_Masks, texture=/Game/Global/Textures/Default/Noise/T_Metal_Noise}.R
    B=ScalarParameter() {parameterName=HeatHitFactor}
   )
  )
  n18 = NamedRerouteUsage()
  n19 = Add(
   A=Add(A=Multiply(A=@n8, B=@n9), B=Multiply(A=@n11, B=@n12))
   B=Multiply(A=Multiply(A=@n17, B=@n18), B=ScalarParameter() {defaultValue=1, group=WingBath, parameterName=HeatTransition_Bar})
  )
  NamedRerouteDeclaration(None=Saturate(None=@n19)) {name=HeatMsk, nodeColor=<rgba 1 0 0.75 1>}
  n20 = TextureSample(UVs=@n19) {
   samplerSource=SSM_Clamp_WorldGroupSettings
   texture=/BRCosmetics/Animation/Props/Gadgets_WingBath/Textures/T_WingBath_Emote_Prop_Heat_Gradient
  }
  NamedRerouteDeclaration(None=Multiply(A=@n20.RGB, B=ScalarParameter() {defaultValue=1, group=WingBath, parameterName=HeatIntensity, sortPriority=0})) {name=HeatEmissive, nodeColor=<rgba 0.9922 0 1 1>}
 ]
}
